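# Niagara ribbon contrail effect for the Red Riding variant of the VinderTech glider.
# A single GreenTrail emitter spawns anim trails from two socket or bone pairs and renders them as narrow ribbons.
# Front-end switches vary color, lifetime, offset and velocity, adding drift and a fainter alpha in menus.
# In gameplay, a curve driven by the system Timeline time scales alpha, while in the front end that scale stays constant.
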NiagaraSystem /BRCosmetics/Gadgets/Assets/VinderTech_GliderChute/Glider_Red_Riding/FX/NS_GliderContrail_RedRiding_Converted
dependencies {
 usedDataInterfaces [&NiagaraDataInterfaceCurve, &NiagaraDataInterfaceSocketReader, &NiagaraDataInterfaceArrayFloat]
 usedDynamicInputs [
  &Boolean_OrOperation
  &SetBoolByFloatComparison
  &RemapRange
  &SelectFloatFromArray
  &MakeIntFromFloat
 ]
 usedModules [
  &self:SystemSpawnScript.SystemScriptSource.SystemScriptGraph.NiagaraNodeAssignment_0.SetVariables_#1
  &SystemState
  &Timeline
  &FN_Integrator
  &InterpolateOverTime
 ]
}
emitters {GreenTrail {
  emitterUpdateScript [
   &EmitterState()
   &FN_SpawnAnimTrail()
  ]
  particleSpawnScript [
   &FN_InitializeAnimTrailBoneIndecies(First Bone or Socket Index=1, Second Bone or Socket Index=0)
   &FN_InitializeAnimTrailBoneIndecies(AnimTrail Selection=1, First Bone or Socket Index=3, Second Bone or Socket Index=2) {module=FN_InitializeAnimTrailBoneIndecies001}
   &FN_InitializeAnimTrail(
    AnimTrailWidth=1.2
    Color=&MakeCustomLinearColorFromBool(Bool=$User.bIsFrontEnd, True Color=<rgba 5 5 5 0.2>, False Color=<rgba 5 5 5 1>)
    Lifetime=&FN_FrontEndFloatModifier(Gameplay=0.6, Front End=0.9)
    Position Offset=&FN_FrontEndVector3Modifier(Gameplay=<0 0 3>, Front End=<0 0 5>)
   )
   &AddVelocity(Velocity=&FN_FrontEndVector3Modifier(Gameplay=<0 0 0>, Front End=<350 0 0>))
  ]
  particleUpdateScript [
   &ParticleState()
   &FN_UpdateAnimTrail()
   &ScaleColor(
    Scale Alpha=&FN_FrontEndFloatModifier(
     Gameplay=&FloatFromCurve(
      FloatCurve=DataInterface() {propertyValues {Curve {keys [{interpMode=RCIM_Cubic, leaveTangent=-3.4763, tangentMode=RCTM_Break, time=0, value=1}, <at 1 0>]}}}
      Scale Curve=$System.Timeline.Time
     )
     Front End=1
    )
    ScaleA=true
    ScaleRGB=true
   ) {module=ScaleColor002}
   &SolveForcesAndVelocity(Clamp Velocity=false, Limit Acceleration=false)
  ]
  rendererClasses [&NiagaraRibbonRendererProperties]
  renderers [{rendererClass=&NiagaraRibbonRendererProperties}]
 }}
summary {
 systemName=NS_GliderContrail_RedRiding_Converted
 emitters [{emitterName=GreenTrail, rendererClasses [&NiagaraRibbonRendererProperties]}]
 userVariables [
  User.bIsFrontEnd {type=NiagaraBool}
  User.bIsFrontEndPreview {type=NiagaraBool}
  User.ForwardDot {type=NiagaraFloat, defaultValue=1, description=""}
  User.RibbonIDOffset {type=NiagaraInt32}
  User.RightDot {type=NiagaraFloat, defaultValue=0, description=""}
 ]
}
